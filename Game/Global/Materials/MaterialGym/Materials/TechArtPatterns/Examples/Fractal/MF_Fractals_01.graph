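# Material function example that layers an animated, glowing fractal pattern over an input material, blended by an Alpha input.
# Tiled mesh UVs and object-stabilized screen-space UVs are lerped, then wrapped with Frac.
# A Custom HLSL node loops over repeated UV cells, summing time-animated sine rings tinted by a cosine palette.
# The result drives emissive, while parameters set base color, metallic, specular and roughness before BlendMaterialAttributes.
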MaterialFunction /Game/Global/Materials/MaterialGym/Materials/TechArtPatterns/Examples/Fractal/MF_Fractals_01
graph {nodes [
  n1 = FunctionInput(Preview=SetMaterialAttributes!(MaterialAttributes)) {bUsePreviewValueAsDefault=true, inputName=Material, inputType=FunctionInput_MaterialAttributes}
  n2 = Multiply(A=TextureCoordinate(), B=ScalarParameter() {defaultValue=6, group=Fractal, parameterName=Tile})
  n3 = VectorParameter() {group=Fractal, parameterName=FractalUVOffset, defaultValue=<rgba 0 0 0 1>}
  n4 = ComponentMask(None=@n3.RGB) {g=true, r=true}
  n5 = Add(A=ObjectPositionWS(), B=@n3.RGB)
  n6 = ScalarParameter() {defaultValue=0.25, group=Fractal, parameterName=UVScale}
  MaterialExpressionFrac_1 = Frac(None=LinearInterpolate(
    A=Add(A=@n2, B=@n4)
    B=MaterialFunctionCall!(Position=@n5, UVRadialScale=@n6) {materialFunction=/Game/Weapons/WeaponSkins/MaterialFunction/ObjectStabilizedScreenSpaceUVs}.StabilizedUVs
    Alpha=ScalarParameter() {group=Fractal, parameterName=ScreenUVs}
   ))
  MaterialExpressionTime_1 = Time()
  MaterialExpressionScalarParameter_1 = ScalarParameter() {defaultValue=0.01, group=Fractal, parameterName=Numerator}
  MaterialExpressionScalarParameter_6 = ScalarParameter() {defaultValue=1.75, group=Fractal, parameterName=exponent}
  MaterialExpressionScalarParameter_9 = ScalarParameter() {defaultValue=4, parameterName=iterations}
  n7 = Custom!(
   uv=@MaterialExpressionFrac_1
   time=@MaterialExpressionTime_1
   x=@MaterialExpressionScalarParameter_1
   y=@MaterialExpressionScalarParameter_6
   iterations=@MaterialExpressionScalarParameter_9
  ) {
   code="struct Functions\r\n{\r\nfloat3 Palette(float t)\r\n{\r\n    float3 a = float3(0.5, 0.5, 0.5);\r\n    float3 b = float3(0.5, 0.5, 0.5);\r\n    float3 c = float3(1.0, 1.0, 1.0);\r\n    float3 d = float3(0.263, 0.416, 0.557);\r\n\r\n    return a + b * cos(6.28318*(c*t+d));\r\n}\r\nfloat3 Fractal(float2 uv, float time, float x, float y, float iterations)\r\n{\r\n    float3 finalColor = 0.0;\r\n    for (int i = 0; i < int(iterations); i++)\r\n    {\r\n        float2 uv1 = frac(uv * (i+1)) - .5;\r\n        float d = length(uv1) * exp(-length(uv-.5));\r\n        float3 col = Palette(length(uv-.5) + i*.4 + time*.4);\r\n        d = sin(d*8.0 + time)/8.0;\r\n        d = abs(d);\r\n        d = pow(x / d, y);\r\n        finalColor += col * d;\r\n    }\r\n    return finalColor;\r\n}\r\n};\r\nFunctions f;\r\nreturn f.Fractal(uv, time, x, y, iterations);\r\n\r\n\r\n"
   inputs [
    {
     inputName=uv
     input {expression=&self:MaterialExpressionFrac_1}
    }
    {
     inputName=time
     input {expression=&self:MaterialExpressionTime_1}
    }
    {
     inputName=x
     input {expression=&self:MaterialExpressionScalarParameter_1}
    }
    {
     inputName=y
     input {expression=&self:MaterialExpressionScalarParameter_6}
    }
    {
     inputName=iterations
     input {expression=&self:MaterialExpressionScalarParameter_9}
    }
   ]
  }
  n8 = SetMaterialAttributes!(
   MaterialAttributes=@n1
   BaseColor=VectorParameter() {parameterName=FractalBaseColor, defaultValue=<rgba 0 0 0 1>}.RGB
   Metallic=ScalarParameter() {defaultValue=0.4, group=Fractal, parameterName=FractalMetallic}
   Specular=ScalarParameter() {defaultValue=0.4, group=Fractal, parameterName=FractalSpec}
   Roughness=ScalarParameter() {defaultValue=0.4, group=Fractal, parameterName=FractalRough}
   EmissiveColor=Multiply(A=@n7, B=ScalarParameter() {defaultValue=1, group=Fractal, parameterName=FractalEmissiveStrength})
  ) {attributeSetTypes [
    #1
    #2
    #3
    #4
    #5
   ]}
  n9 = FunctionInput() {bUsePreviewValueAsDefault=true, inputName=Alpha, inputType=FunctionInput_Scalar, sortPriority=1, previewValue=<1 0 0 1>}
  FunctionOutput(None=BlendMaterialAttributes(A=@n1, B=@n8, Alpha=@n9))
 ]}
