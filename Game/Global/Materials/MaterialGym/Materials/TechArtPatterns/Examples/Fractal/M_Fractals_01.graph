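# M_Fractals_01 is a Default Lit material example from the TechArtPatterns fractal set in the Material Gym.
# It is a thin wrapper around the MF_Fractals_01 material function, which exposes Material and Alpha outputs and holds the fractal logic.
# The material uses material attributes and is flagged for skeletal mesh and mesh deformer use, with automatic usage setting in the editor turned off.
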Material /Game/Global/Materials/MaterialGym/Materials/TechArtPatterns/Examples/Fractal/M_Fractals_01
graph {
 properties {
  bAutomaticallySetUsageInEditor=false
  bUseMaterialAttributes=true
  bUsedWithMeshDeformer=true
  bUsedWithSkeletalMesh=true
  pixelDepthOffsetMode=PDOM_Legacy
  usedShadingModels=Default Lit
 }
 nodes [MaterialFunctionCall!(Material, Alpha) {materialFunction=&/MF_Fractals_01}]
}
